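# Sound cue for synth bass voice 02 of the Creative toy keyboard, playing one note per key with music-block attenuation.
# A KeysIndex switch, with index 0 empty, selects root samples C1, C2, C_1 or C_2, or fixed-pitch copies of them.
# Equal-tempered modulator ratios derive seven-note major scales from C1/C2 and five-note sets from C_1/C_2.
# All quality levels share that switch, and concurrency stops the oldest voice.
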SoundCue /Game/Sounds/Creative/Toys/Keyboard/Bass/CM_Keyboard_Synth_Bass_02_Cue
root=QualityLevel_0
nodes [
 {id=Modulator_10, children [WavePlayer_8], properties {pitchMax=1.1225, pitchMin=1.1225, volumeMax=1, volumeMin=1}}
 {id=Modulator_11, children [WavePlayer_9], properties {pitchMax=1.2599, pitchMin=1.2599}}
 {id=Modulator_12, children [WavePlayer_10], properties {pitchMax=1.3348, pitchMin=1.3348}}
 {id=Modulator_13, children [WavePlayer_11], properties {pitchMax=1.4983, pitchMin=1.4983}}
 {id=Modulator_14, children [WavePlayer_12], properties {pitchMax=1.6818, pitchMin=1.6818}}
 {id=Modulator_15, children [WavePlayer_13], properties {pitchMax=1.8877, pitchMin=1.8877}}
 {id=Modulator_16, children [WavePlayer_15], properties {pitchMax=1.1225, pitchMin=1.1225, volumeMax=1, volumeMin=1}}
 {id=Modulator_17, children [WavePlayer_16], properties {pitchMax=1.3348, pitchMin=1.3348}}
 {id=Modulator_18, children [WavePlayer_17], properties {pitchMax=1.4983, pitchMin=1.4983}}
 {id=Modulator_19, children [WavePlayer_18], properties {pitchMax=1.6818, pitchMin=1.6818}}
 {id=Modulator_24, children [WavePlayer_20], properties {pitchMax=1.1225, pitchMin=1.1225, volumeMax=1, volumeMin=1}}
 {id=Modulator_25, children [WavePlayer_21], properties {pitchMax=1.3348, pitchMin=1.3348}}
 {id=Modulator_26, children [WavePlayer_22], properties {pitchMax=1.4983, pitchMin=1.4983}}
 {id=Modulator_27, children [WavePlayer_23], properties {pitchMax=1.6818, pitchMin=1.6818}}
 {id=Modulator_3, children [WavePlayer_1], properties {pitchMax=1.1225, pitchMin=1.1225, volumeMax=1, volumeMin=1}}
 {id=Modulator_4, children [WavePlayer_2], properties {pitchMax=1.2599, pitchMin=1.2599}}
 {id=Modulator_6, children [WavePlayer_3], properties {pitchMax=1.3348, pitchMin=1.3348}}
 {id=Modulator_7, children [WavePlayer_4], properties {pitchMax=1.4983, pitchMin=1.4983}}
 {id=Modulator_8, children [WavePlayer_5], properties {pitchMax=1.6818, pitchMin=1.6818}}
 {id=Modulator_9, children [WavePlayer_6], properties {pitchMax=1.8877, pitchMin=1.8877}}
 {id=QualityLevel_0, children [Switch_1,Switch_1,Switch_1,Switch_1]}
 {
  id=Switch_1
  children [
   null
   WavePlayer_0
   Modulator_3
   Modulator_4
   Modulator_6
   Modulator_7
   Modulator_8
   Modulator_9
   WavePlayer_7
   Modulator_10
   Modulator_11
   Modulator_12
   Modulator_13
   Modulator_14
   Modulator_15
   WavePlayer_14
   Modulator_16
   Modulator_17
   Modulator_18
   Modulator_19
   WavePlayer_19
   Modulator_24
   Modulator_25
   Modulator_26
   Modulator_27
  ]
  properties {intParameterName=KeysIndex}
 }
 {id=WavePlayer_0, properties {soundWaveAssetPtr=&/CM_Synth_Bass_02_C1}}
 {id=WavePlayer_1, properties {soundWaveAssetPtr=&CM_Synth_Bass_02_C1}}
 {id=WavePlayer_10, properties {soundWaveAssetPtr=&/CM_Synth_Bass_02_C2}}
 {id=WavePlayer_11, properties {soundWaveAssetPtr=&CM_Synth_Bass_02_C2}}
 {id=WavePlayer_12, properties {soundWaveAssetPtr=&CM_Synth_Bass_02_C2}}
 {id=WavePlayer_13, properties {soundWaveAssetPtr=&CM_Synth_Bass_02_C2}}
 {id=WavePlayer_14, properties {soundWaveAssetPtr=&/CM_Synth_Bass_02_C_1}}
 {id=WavePlayer_15, properties {soundWaveAssetPtr=&CM_Synth_Bass_02_C_1}}
 {id=WavePlayer_16, properties {soundWaveAssetPtr=&CM_Synth_Bass_02_C_1}}
 {id=WavePlayer_17, properties {soundWaveAssetPtr=&CM_Synth_Bass_02_C_1}}
 {id=WavePlayer_18, properties {soundWaveAssetPtr=&CM_Synth_Bass_02_C_1}}
 {id=WavePlayer_19, properties {soundWaveAssetPtr=&/CM_Synth_Bass_02_C_2}}
 {id=WavePlayer_2, properties {soundWaveAssetPtr=&CM_Synth_Bass_02_C1}}
 {id=WavePlayer_20, properties {soundWaveAssetPtr=&CM_Synth_Bass_02_C_2}}
 {id=WavePlayer_21, properties {soundWaveAssetPtr=&CM_Synth_Bass_02_C_2}}
 {id=WavePlayer_22, properties {soundWaveAssetPtr=&CM_Synth_Bass_02_C_2}}
 {id=WavePlayer_23, properties {soundWaveAssetPtr=&CM_Synth_Bass_02_C_2}}
 {id=WavePlayer_3, properties {soundWaveAssetPtr=&CM_Synth_Bass_02_C1}}
 {id=WavePlayer_4, properties {soundWaveAssetPtr=&CM_Synth_Bass_02_C1}}
 {id=WavePlayer_5, properties {soundWaveAssetPtr=&CM_Synth_Bass_02_C1}}
 {id=WavePlayer_6, properties {soundWaveAssetPtr=&CM_Synth_Bass_02_C1}}
 {id=WavePlayer_7, properties {soundWaveAssetPtr=&CM_Synth_Bass_02_C2}}
 {id=WavePlayer_8, properties {soundWaveAssetPtr=&CM_Synth_Bass_02_C2}}
 {id=WavePlayer_9, properties {soundWaveAssetPtr=&CM_Synth_Bass_02_C2}}
]
properties {
 attenuationSettings=/Game/Sounds/Creative/Attenuations/Fort_Creative_MusicBlocks_Att
 bPrimeOnLoad=true
 maxDistance=100000
 soundClassObject=&SFX
 volumeMultiplier=1.3
 concurrencyOverrides {maxCount=8, resolutionRule=StopOldest}
}
