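# Sound cue for the third synth bass voice of the Creative toy keyboard, using music-block attenuation and capped concurrency that stops the oldest note.
# A switch on the KeysIndex parameter leaves index 0 silent and maps the rest to 24 notes.
# Fixed-pitch modulators derive the notes from four samples: two seven-note major-scale octaves, then two five-note sets spaced like black keys.
# All quality levels share that switch.
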SoundCue /Game/Sounds/Creative/Toys/Keyboard/Bass/CM_Keyboard_Synth_Bass_03_Cue
root=QualityLevel_0
nodes [
 {id=Modulator_0, children [WavePlayer_25], properties {pitchMax=1.1225, pitchMin=1.1225, volumeMax=1, volumeMin=1}}
 {id=Modulator_1, children [WavePlayer_26], properties {pitchMax=1.2599, pitchMin=1.2599}}
 {id=Modulator_2, children [WavePlayer_27], properties {pitchMax=1.3348, pitchMin=1.3348}}
 {id=Modulator_20, children [WavePlayer_29], properties {pitchMax=1.6817, pitchMin=1.6817}}
 {id=Modulator_21, children [WavePlayer_30], properties {pitchMax=1.8877, pitchMin=1.8877}}
 {id=Modulator_22, children [WavePlayer_32], properties {pitchMax=1.1225, pitchMin=1.1225, volumeMax=1, volumeMin=1}}
 {id=Modulator_23, children [WavePlayer_33], properties {pitchMax=1.2599, pitchMin=1.2599}}
 {id=Modulator_28, children [WavePlayer_34], properties {pitchMax=1.3348, pitchMin=1.3348}}
 {id=Modulator_29, children [WavePlayer_35], properties {pitchMax=1.4983, pitchMin=1.4983}}
 {id=Modulator_30, children [WavePlayer_36], properties {pitchMax=1.6817, pitchMin=1.6817}}
 {id=Modulator_31, children [WavePlayer_37], properties {pitchMax=1.8877, pitchMin=1.8877}}
 {id=Modulator_32, children [WavePlayer_40], properties {pitchMax=1.1225, pitchMin=1.1225, volumeMax=1, volumeMin=1}}
 {id=Modulator_33, children [WavePlayer_41], properties {pitchMax=1.3348, pitchMin=1.3348}}
 {id=Modulator_34, children [WavePlayer_42], properties {pitchMax=1.4983, pitchMin=1.4983}}
 {id=Modulator_35, children [WavePlayer_43], properties {pitchMax=1.6818, pitchMin=1.6818}}
 {id=Modulator_36, children [WavePlayer_45], properties {pitchMax=1.1225, pitchMin=1.1225, volumeMax=1, volumeMin=1}}
 {id=Modulator_37, children [WavePlayer_46], properties {pitchMax=1.3348, pitchMin=1.3348}}
 {id=Modulator_38, children [WavePlayer_47], properties {pitchMax=1.4983, pitchMin=1.4983}}
 {id=Modulator_39, children [WavePlayer_48], properties {pitchMax=1.6818, pitchMin=1.6818}}
 {id=Modulator_5, children [WavePlayer_28], properties {pitchMax=1.4983, pitchMin=1.4983}}
 {id=QualityLevel_0, children [Switch_1,Switch_1,Switch_1,Switch_1]}
 {
  id=Switch_1
  children [
   null
   WavePlayer_24
   Modulator_0
   Modulator_1
   Modulator_2
   Modulator_5
   Modulator_20
   Modulator_21
   WavePlayer_31
   Modulator_22
   Modulator_23
   Modulator_28
   Modulator_29
   Modulator_30
   Modulator_31
   WavePlayer_38
   Modulator_32
   Modulator_33
   Modulator_34
   Modulator_35
   WavePlayer_44
   Modulator_36
   Modulator_37
   Modulator_38
   Modulator_39
  ]
  properties {intParameterName=KeysIndex}
 }
 {id=WavePlayer_24, properties {soundWaveAssetPtr=&/CM_Synth_Bass_03_C1}}
 {id=WavePlayer_25, properties {soundWaveAssetPtr=&CM_Synth_Bass_03_C1}}
 {id=WavePlayer_26, properties {soundWaveAssetPtr=&CM_Synth_Bass_03_C1}}
 {id=WavePlayer_27, properties {soundWaveAssetPtr=&CM_Synth_Bass_03_C1}}
 {id=WavePlayer_28, properties {soundWaveAssetPtr=&CM_Synth_Bass_03_C1}}
 {id=WavePlayer_29, properties {soundWaveAssetPtr=&CM_Synth_Bass_03_C1}}
 {id=WavePlayer_30, properties {soundWaveAssetPtr=&CM_Synth_Bass_03_C1}}
 {id=WavePlayer_31, properties {soundWaveAssetPtr=&/CM_Synth_Bass_03_C2}}
 {id=WavePlayer_32, properties {soundWaveAssetPtr=&CM_Synth_Bass_03_C2}}
 {id=WavePlayer_33, properties {soundWaveAssetPtr=&CM_Synth_Bass_03_C2}}
 {id=WavePlayer_34, properties {soundWaveAssetPtr=&CM_Synth_Bass_03_C2}}
 {id=WavePlayer_35, properties {soundWaveAssetPtr=&CM_Synth_Bass_03_C2}}
 {id=WavePlayer_36, properties {soundWaveAssetPtr=&CM_Synth_Bass_03_C2}}
 {id=WavePlayer_37, properties {soundWaveAssetPtr=&CM_Synth_Bass_03_C2}}
 {id=WavePlayer_38, properties {soundWaveAssetPtr=&/CM_Synth_Bass_03_C_1}}
 {id=WavePlayer_40, properties {soundWaveAssetPtr=&CM_Synth_Bass_03_C_1}}
 {id=WavePlayer_41, properties {soundWaveAssetPtr=&CM_Synth_Bass_03_C_1}}
 {id=WavePlayer_42, properties {soundWaveAssetPtr=&CM_Synth_Bass_03_C_1}}
 {id=WavePlayer_43, properties {soundWaveAssetPtr=&CM_Synth_Bass_03_C_1}}
 {id=WavePlayer_44, properties {soundWaveAssetPtr=&/CM_Synth_Bass_03_C_2}}
 {id=WavePlayer_45, properties {soundWaveAssetPtr=&CM_Synth_Bass_03_C_2}}
 {id=WavePlayer_46, properties {soundWaveAssetPtr=&CM_Synth_Bass_03_C_2}}
 {id=WavePlayer_47, properties {soundWaveAssetPtr=&CM_Synth_Bass_03_C_2}}
 {id=WavePlayer_48, properties {soundWaveAssetPtr=&CM_Synth_Bass_03_C_2}}
]
properties {
 attenuationSettings=/Game/Sounds/Creative/Attenuations/Fort_Creative_MusicBlocks_Att
 bPrimeOnLoad=true
 maxDistance=100000
 soundClassObject=&SFX
 volumeMultiplier=1.4
 concurrencyOverrides {maxCount=8, resolutionRule=StopOldest}
}
