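# Sound cue for the mobile variant of synth bass 03 on the Creative toy keyboard, playing one note per KeysIndex value.
# A Switch leaves index 0 silent and maps the rest to four C-named samples, played raw or through fixed-pitch Modulators that derive other scale notes from each.
# All quality levels share that Switch.
# Music-block attenuation applies, and concurrency caps at four voices, stopping the oldest.
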SoundCue /Game/Sounds/Creative/Toys/Keyboard/Bass/CM_Keyboard_Synth_Bass_03_Mobile_Cue
root=QualityLevel_0
nodes [
 {id=Modulator_10, children [WavePlayer_32], properties {pitchMax=1.1225, pitchMin=1.1225, volumeMax=1, volumeMin=1}}
 {id=Modulator_11, children [WavePlayer_33], properties {pitchMax=1.2599, pitchMin=1.2599}}
 {id=Modulator_12, children [WavePlayer_34], properties {pitchMax=1.3348, pitchMin=1.3348}}
 {id=Modulator_13, children [WavePlayer_35], properties {pitchMax=1.4983, pitchMin=1.4983}}
 {id=Modulator_14, children [WavePlayer_36], properties {pitchMax=1.6817, pitchMin=1.6817}}
 {id=Modulator_15, children [WavePlayer_37], properties {pitchMax=1.8877, pitchMin=1.8877}}
 {id=Modulator_16, children [WavePlayer_40], properties {pitchMax=1.1225, pitchMin=1.1225, volumeMax=1, volumeMin=1}}
 {id=Modulator_17, children [WavePlayer_41], properties {pitchMax=1.3348, pitchMin=1.3348}}
 {id=Modulator_18, children [WavePlayer_42], properties {pitchMax=1.4983, pitchMin=1.4983}}
 {id=Modulator_19, children [WavePlayer_43], properties {pitchMax=1.6817, pitchMin=1.6817}}
 {id=Modulator_24, children [WavePlayer_45], properties {pitchMax=1.1225, pitchMin=1.1225, volumeMax=1, volumeMin=1}}
 {id=Modulator_25, children [WavePlayer_46], properties {pitchMax=1.3348, pitchMin=1.3348}}
 {id=Modulator_26, children [WavePlayer_47], properties {pitchMax=1.4983, pitchMin=1.4983}}
 {id=Modulator_27, children [WavePlayer_48], properties {pitchMax=1.6817, pitchMin=1.6817}}
 {id=Modulator_3, children [WavePlayer_25], properties {pitchMax=1.1225, pitchMin=1.1225, volumeMax=1, volumeMin=1}}
 {id=Modulator_4, children [WavePlayer_26], properties {pitchMax=1.2599, pitchMin=1.2599}}
 {id=Modulator_6, children [WavePlayer_27], properties {pitchMax=1.3348, pitchMin=1.3348}}
 {id=Modulator_7, children [WavePlayer_28], properties {pitchMax=1.4983, pitchMin=1.4983}}
 {id=Modulator_8, children [WavePlayer_29], properties {pitchMax=1.6817, pitchMin=1.6817}}
 {id=Modulator_9, children [WavePlayer_30], properties {pitchMax=1.8877, pitchMin=1.8877}}
 {id=QualityLevel_0, children [Switch_1,Switch_1,Switch_1,Switch_1]}
 {
  id=Switch_1
  children [
   null
   WavePlayer_24
   Modulator_3
   Modulator_4
   Modulator_6
   Modulator_7
   Modulator_8
   Modulator_9
   WavePlayer_31
   Modulator_10
   Modulator_11
   Modulator_12
   Modulator_13
   Modulator_14
   Modulator_15
   WavePlayer_38
   Modulator_16
   Modulator_17
   Modulator_18
   Modulator_19
   WavePlayer_44
   Modulator_24
   Modulator_25
   Modulator_26
   Modulator_27
  ]
  properties {intParameterName=KeysIndex}
 }
 {id=WavePlayer_24, properties {soundWaveAssetPtr=&/CM_Synth_Bass_03_C1}}
 {id=WavePlayer_25, properties {soundWaveAssetPtr=&CM_Synth_Bass_03_C1}}
 {id=WavePlayer_26, properties {soundWaveAssetPtr=&CM_Synth_Bass_03_C1}}
 {id=WavePlayer_27, properties {soundWaveAssetPtr=&CM_Synth_Bass_03_C1}}
 {id=WavePlayer_28, properties {soundWaveAssetPtr=&CM_Synth_Bass_03_C1}}
 {id=WavePlayer_29, properties {soundWaveAssetPtr=&CM_Synth_Bass_03_C1}}
 {id=WavePlayer_30, properties {soundWaveAssetPtr=&CM_Synth_Bass_03_C1}}
 {id=WavePlayer_31, properties {soundWaveAssetPtr=&/CM_Synth_Bass_03_C2}}
 {id=WavePlayer_32, properties {soundWaveAssetPtr=&CM_Synth_Bass_03_C2}}
 {id=WavePlayer_33, properties {soundWaveAssetPtr=&CM_Synth_Bass_03_C2}}
 {id=WavePlayer_34, properties {soundWaveAssetPtr=&CM_Synth_Bass_03_C2}}
 {id=WavePlayer_35, properties {soundWaveAssetPtr=&CM_Synth_Bass_03_C2}}
 {id=WavePlayer_36, properties {soundWaveAssetPtr=&CM_Synth_Bass_03_C2}}
 {id=WavePlayer_37, properties {soundWaveAssetPtr=&CM_Synth_Bass_03_C2}}
 {id=WavePlayer_38, properties {soundWaveAssetPtr=&/CM_Synth_Bass_03_C_1}}
 {id=WavePlayer_40, properties {soundWaveAssetPtr=&CM_Synth_Bass_03_C_1}}
 {id=WavePlayer_41, properties {soundWaveAssetPtr=&CM_Synth_Bass_03_C_1}}
 {id=WavePlayer_42, properties {soundWaveAssetPtr=&CM_Synth_Bass_03_C_1}}
 {id=WavePlayer_43, properties {soundWaveAssetPtr=&CM_Synth_Bass_03_C_1}}
 {id=WavePlayer_44, properties {soundWaveAssetPtr=&/CM_Synth_Bass_03_C_2}}
 {id=WavePlayer_45, properties {soundWaveAssetPtr=&CM_Synth_Bass_03_C_2}}
 {id=WavePlayer_46, properties {soundWaveAssetPtr=&CM_Synth_Bass_03_C_2}}
 {id=WavePlayer_47, properties {soundWaveAssetPtr=&CM_Synth_Bass_03_C_2}}
 {id=WavePlayer_48, properties {soundWaveAssetPtr=&CM_Synth_Bass_03_C_2}}
]
properties {
 attenuationSettings=/Game/Sounds/Creative/Attenuations/Fort_Creative_MusicBlocks_Att
 bOverrideConcurrency=true
 bPrimeOnLoad=true
 maxDistance=100000
 soundClassObject=&SFX
 volumeMultiplier=1.4
 concurrencyOverrides {maxCount=4, resolutionRule=StopOldest}
}
